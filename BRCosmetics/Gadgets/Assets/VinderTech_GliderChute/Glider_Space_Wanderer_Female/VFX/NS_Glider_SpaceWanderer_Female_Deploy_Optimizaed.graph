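# Niagara effect for the female Space Wanderer glider, layering blue-purple thruster meshes, fire, GPU ash, flares, lights and speedlines.
# Most emitters follow skeletal mesh sockets and scale by a shared AccelerationRamp; the large flare integrates its own from User.ForwardDot.
# Purple_Main001 merges the disabled Blue_Main and Purple_Main using spawn groups and per-group scales for two mesh renderers.
# Front-end modifiers swap selected sizes, offsets, velocities and rates outside gameplay.
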NiagaraSystem /BRCosmetics/Gadgets/Assets/VinderTech_GliderChute/Glider_Space_Wanderer_Female/VFX/NS_Glider_SpaceWanderer_Female_Deploy_Optimizaed
dependencies {
 usedDataInterfaces [
  &NiagaraDataInterfaceVectorCurve
  &NiagaraDataInterfaceCurve
  &NiagaraDataInterfaceColorCurve
  &NiagaraDataInterfaceVector2DCurve
 ]
 usedDynamicInputs []
 usedModules [
  &self:SystemSpawnScript.SystemScriptSource.SystemScriptGraph.NiagaraNodeAssignment_1.SetVariables_#1
  &SystemState
  &self:SystemSpawnScript.SystemScriptSource.SystemScriptGraph.NiagaraNodeAssignment_0.SetVariables_#2
 ]
}
emitters {
 Ash {
  simTarget=GPUComputeSim
  emitterUpdateScript [
   /SetVariables(Emitter.AccelerationRamp:NiagaraFloat=$System.AccelerationRamp) {isSetParameters=true, module=SetVariables_#3}
   &EmitterState(Inactive Response=::NewEnumerator1, Life Cycle Mode=::NewEnumerator1, Loop Duration=1, UseLoopDelay=false)
   &SpawnRate(SpawnRate=&Lerp_Float(A=30, B=60, Alpha=$Emitter.AccelerationRamp), Use Spawn Probability=false)
  ]
  particleSpawnScript [
   /Niagara/Modules/Spawn/Initialization/InitializeParticle(
    Color=&LerpLinearColors(
     StartColor=&MakeLinearColorFromVectorAndFloat(
      "Vector (RGB)"=&MakeVectorFromLinearColorRGB(Color=<rgba 0 0.0962 1 1>)
      "Float (Alpha)"=&UniformRangedFloat(Minimum=0.01, Maximum=0.1)
     )
     EndColor=&MakeLinearColorFromVectorAndFloat(
      "Vector (RGB)"=&MakeVectorFromLinearColorRGB(Color=<rgba 0 0.8097 1 1>)
      "Float (Alpha)"=&UniformRangedFloat(Minimum=0.01, Maximum=0.1)
     )
     LerpFactor=&Power(a=&UniformRangedFloat(Minimum=0, Maximum=1), b=1.3333)
    )
    Lifetime=&Lerp_Float(
     A=&UniformRangedFloat(Minimum=0.6, Maximum=1)
     B=&UniformRangedFloat(Minimum=0.3, Maximum=0.7)
     Alpha=$Emitter.AccelerationRamp
    )
    Module.Write Color=true
    Module.Write Lifetime=true
    Module.Write Mass=false
    Module.Write Material Random=true
    Module.Write Position=false
    Module.Write Scale=false
    Module.Write SpriteRotation=true
    Module.Write SpriteSize=false
    Position=&SimulationPosition()
    Sprite Rotation=&UniformRangedFloat(Minimum=0, Maximum=359)
   )
   &SkeletalMeshLocation(
    Bone / Socket Index=1
    Inherit Velocity=false
    MaskSpawnGroup=false
    Mesh Sampling Type=::NewEnumerator5
    Sampled Position Offset=<0 0 -15>
    Skeletal Mesh=$User.SkeletalMesh
    Socket Sampling Mode=::NewEnumerator5
   ) {enabled=false}
   &SampleSkeletalMeshSkeleton(Mesh:NiagaraDataInterfaceSkeletalMesh=$User.SkeletalMesh)
   /Niagara/Modules/Spawn/Location/SkeletalMeshSkeletonLocation(Module.Inherit Velocity=false, Module.Use Sampled Bone as Ribbon ID=true)
   /Niagara/Modules/Spawn/Location/SphereLocation(
    Hemisphere Y=true
    MaskSpawnGroup=false
    Module.Override Randomness=false
    Non Uniform Scale=<1.333 2.25 1>
    Offset=<-5 0 -10>
    Sphere Radius=&FN_FrontEndFloatModifier(Gameplay=&Lerp_Float(A=47.5, B=35, Alpha=$Emitter.AccelerationRamp), Front End=50)
    Surface Only=true
    Surface Only Band Thickness=0.5
   )
   /SetVariables(Particles.SpriteSize=&Vector2DFromFloat(Value=&Lerp_Float(
      A=&Lerp_Float(A=7, B=14, Alpha=&Power(a=$Particles.MaterialRandom, b=2))
      B=&Lerp_Float(A=7, B=14, Alpha=&Power(a=$Particles.MaterialRandom, b=0.75))
      Alpha=$Emitter.AccelerationRamp
     ))) {isSetParameters=true}
   &AddVelocity(Velocity=&FN_FrontEndVector3Modifier(
     Gameplay=&Lerp_Vector(
      A=&UniformRangedVector(Minimum=<-1 30 -1>, Maximum=<1 120 1>)
      B=&UniformRangedVector(Minimum=<-1 360 -1>, Maximum=<1 900 1>)
      Alpha=$Emitter.AccelerationRamp
     )
     Front End=&UniformRangedVector(Minimum=<-1 -500 -1>, Maximum=<1 -900 1>)
    ))
  ]
  particleUpdateScript [
   &ParticleState()
   &SolveForcesAndVelocity(Clamp Velocity=false, Limit Acceleration=false)
   /Niagara/Modules/Update/SubUV/SubUVAnimation(Number Of Frames=16)
   &ScaleSpriteSize(Scale Factor=&Vector2DFromFloat(Value=&FloatFromCurve(FloatCurve=DataInterface() {propertyValues {Curve {keys [{arriveTangent=4.6425, interpMode=RCIM_Cubic, leaveTangent=-1.1064, tangentMode=RCTM_User, time=0, value=1.25}, <at 1 0>]}}})))
   &CameraOffset(Camera Offset Amount=15)
   &ScaleColor(
    Scale RGB=&VectorFromFloat(Value=&FloatFromCurve(FloatCurve=DataInterface() {propertyValues {Curve {keys [{arriveTangent=-5.1478, interpMode=RCIM_Cubic, leaveTangent=-5.1478, tangentMode=RCTM_User, time=0, value=5}, <at 1 0 RCIM_Cubic>]}}}))
    ScaleA=true
    ScaleRGB=true
   )
   /Game/Effects/Niagara/Modules/Position/FN_OffsetPositionByOrientationAxis() {enabled=false, module=OffsetPositionByMeshOrientation}
  ]
  rendererClasses [&NiagaraSpriteRendererProperties]
  renderers [{rendererClass=&NiagaraSpriteRendererProperties}]
 }
 Blue_Decal {
  emitterUpdateScript [
   /SetVariables(Emitter.AccelerationRamp:NiagaraFloat=$System.AccelerationRamp) {isSetParameters=true, module=SetVariables_#3}
   &EmitterState(Enable Distance Culling=true, MaxDistance=1024, Module.bUseMaxDistance=true, Module.bUseMinDistance=false, Scalability Mode=::NewEnumerator1)
   &SpawnRate(SpawnRate=9, Use Spawn Probability=false)
  ]
  particleSpawnScript [
   &InitializeParticle(
    Color=&MakeLinearColorFromVectorAndFloat(
     "Vector (RGB)"=&MakeVectorFromLinearColorRGB(Color=&LerpLinearColors(StartColor=<rgba 0 2.0243 5 1>, EndColor=<rgba 0 7.2876 18 1>, LerpFactor=$Emitter.AccelerationRamp))
     "Float (Alpha)"=&UniformRangedFloat(Minimum=0.6, Maximum=0.8)
    )
    Lifetime=&UniformRangedFloat(Minimum=0.35, Maximum=0.5)
    Module.Write Color=true
    Module.Write Lifetime=true
    Module.Write Mass=false
    Module.Write Material Random=true
    Module.Write Position=true
    Module.Write Scale=false
    Module.Write SpriteRotation=true
    Module.Write SpriteSize=true
    Position=&Add_Vector(
     A=/Niagara/DynamicInputs/Helpers/SimulationPosition()
     B=&UniformRangedVector(Minimum=<-95 0 -38>, Maximum=<-90 9 -38>)
    )
    Sprite Rotation=&UniformRangedFloat(Minimum=0, Maximum=359)
    Sprite Size=&Vector2DFromFloat(Value=&Lerp_Float(
      A=&UniformRangedFloat(Minimum=65, Maximum=75)
      B=&UniformRangedFloat(Minimum=75, Maximum=85)
      Alpha=$Emitter.AccelerationRamp
     ))
   )
   &DynamicMaterialParameters(
    Index 0 Param 1=0.5
    Module.Param0WriteEnabled=true
    Module.Param1WriteEnabled=false
    Module.Param2WriteEnabled=true
    Module.Param3WriteEnabled=false
   )
  ]
  particleUpdateScript [
   &CameraOffset(Camera Offset Amount=&FloatFromCurve(FloatCurve=DataInterface() {propertyValues {Curve {keys [<at 0 0>, {arriveTangent=-47.8205, leaveTangent=-47.8205, time=1, value=65}]}}}))
   &SkeletalMeshLocation(
    Apply Sampled Orientation to Mesh Particles=true
    Bone / Socket Index=1
    Inherit Velocity=false
    MaskSpawnGroup=false
    Mesh Sampling Type=::NewEnumerator5
    Skeletal Mesh=$User.SkeletalMesh
    Socket Sampling Mode=::NewEnumerator5
   )
   &ParticleState()
   &ScaleSpriteSize(Scale Factor=&Vector2DFromFloat(Value=&FloatFromCurve(FloatCurve=DataInterface() {propertyValues {Curve {keys [<at 0 2 RCIM_Cubic>, {arriveTangent=-3.004, interpMode=RCIM_Cubic, leaveTangent=-3.004, tangentMode=RCTM_User, time=1, value=0.3}]}}})))
   &ScaleColor(
    Scale Alpha=&FloatFromCurve(FloatCurve=DataInterface() {propertyValues {Curve {keys [{arriveTangent=2.7917, interpMode=RCIM_Cubic, leaveTangent=2.7917, tangentMode=RCTM_User, time=0, value=0}, <at 0.9 1 RCIM_Cubic>, <at 1 0>]}}})
    ScaleA=true
    ScaleRGB=true
   )
  ]
  rendererClasses [&NiagaraSpriteRendererProperties]
  renderers [{rendererClass=&NiagaraSpriteRendererProperties}]
 }
 Blue_Main {
  bEnabled=false
  emitterUpdateScript [
   /SetVariables(Emitter.AccelerationRamp:NiagaraFloat=&Divide_Float(A=&Subtract_Float(A=$User.SizeScale, B=0.25), B=0.75)) {enabled=false, isSetParameters=true, module=SetVariables_#4}
   /SetVariables(Emitter.AccelerationRamp:NiagaraFloat=$System.FN_Integrator_v2.Attr) {isSetParameters=true}
   &EmitterState()
   &SpawnRate(SpawnRate=&Lerp_Float(A=8, B=4, Alpha=$Emitter.AccelerationRamp), Use Spawn Probability=false)
  ]
  particleSpawnScript [
   /SetVariables(Particles.MaterialRandom:NiagaraFloat=&UniformRangedFloat(Minimum=0, Maximum=1)) {enabled=false, isSetParameters=true, module=SetVariables_#5}
   &InitializeParticle(
    Color=&LerpLinearColors(StartColor=<rgba 1 1 1 1.333>, EndColor=<rgba 2.5 2.5 2.5 1>, LerpFactor=$Emitter.AccelerationRamp)
    Lifetime=&Lerp_Float(A=0.7, B=1, Alpha=$Emitter.AccelerationRamp)
    Material Random:NiagaraFloat=$Particles.MaterialRandom
    Mesh Scale=&Multiply_Vector(A=<0.8 0.666 0.666>, B=&Lerp_Vector(A=<1.333 1 1>, B=<1.666 0.87 0.87>, Alpha=$System.FN_Integrator_v2.Attr))
    Module.Write Color=true
    Module.Write Lifetime=true
    Module.Write Mass=true
    Module.Write Material Random=true
    Module.Write Position=false
    Module.Write Scale=true
    Module.Write SpriteRotation=true
    Module.Write SpriteSize=false
    Sprite Rotation=&UniformRangedFloat(Minimum=-0.2, Maximum=0.2)
    Sprite Size=<5 5>
   )
  ]
  particleUpdateScript [
   &ParticleState()
   &SampleSkeletalMeshSkeleton(
    Mesh:NiagaraDataInterfaceSkeletalMesh=$User.SkeletalMesh
    Spawn Mode:ENiagaraSkeletalSpawning=::NewEnumerator1
    Specific Bone Index=0
    Specific Socket Index=1
   )
   &SkeletalMeshSkeletonLocation(Module.Inherit Velocity=false, Module.Use Sampled Bone as Ribbon ID=true)
   &FN_OffsetPositionByOrientationAxis(Offset Axis:Vector3f=<65 -4 -10>, Position:NiagaraPosition=$Particles.Position) {module=OffsetPositionByMeshOrientation}
   &ScaleMeshSize(Scale Factor=&Multiply_Vector(
     A=&Lerp_Vector(
      A=&MakeVector(
       X=&FloatFromCurve(FloatCurve=DataInterface() {propertyValues {Curve {keys [<at 0 0.1 RCIM_Cubic>,<at 1 1>]}}})
       Y=&FloatFromCurve(FloatCurve=DataInterface() {propertyValues {Curve {keys [<at 0 1 RCIM_Cubic>,<at 1 0.7>]}}})
       Z=&FloatFromCurve(FloatCurve=DataInterface() {propertyValues {Curve {keys [<at 0 1 RCIM_Cubic>,<at 1 0.7>]}}})
      )
      B=&MakeVector(
       X=&FloatFromCurve(FloatCurve=DataInterface() {propertyValues {Curve {keys [{arriveTangent=-1.7928, interpMode=RCIM_Cubic, time=0, value=1}, {interpMode=RCIM_Cubic, leaveTangent=-0.5551, time=1, value=2}]}}})
       Y=&FloatFromCurve(FloatCurve=DataInterface() {propertyValues {Curve {keys [<at 0 1 RCIM_Cubic>,<at 1 0.7 RCIM_Cubic>]}}})
       Z=&FloatFromCurve(FloatCurve=DataInterface() {propertyValues {Curve {keys [{arriveTangent=-1.5003, interpMode=RCIM_Cubic, time=0, value=1}, {interpMode=RCIM_Cubic, leaveTangent=-2.2772, time=1, value=0.7}]}}})
      )
      Alpha=$Emitter.AccelerationRamp
     )
     B=&VectorFromFloat(Value=&FloatFromCurve(
       FloatCurve=DataInterface() {propertyValues {Curve {keys [
           <at 0 0 RCIM_Constant>
           {arriveTangent=5.5862, interpMode=RCIM_Cubic, leaveTangent=5.5862, tangentMode=RCTM_User, time=0.3, value=0}
           <at 0.7 1>
           <at 1 1>
          ]}}}
       Scale Curve=$Emitter.AccelerationRamp
      ))
    )) {module=ScaleMeshSize001}
   /SetVariables(Particles.Scale=&VectorFromFloat(Value=$System.FN_Integrator_v2.Attr)) {enabled=false, isSetParameters=true}
   &ScaleColor(
    Scale Alpha=&FloatFromCurve(FloatCurve=DataInterface() {propertyValues {Curve {keys [
         <at 0 0>
         {arriveTangent=1.6667, leaveTangent=1.6667, time=0.4, value=1}
         {arriveTangent=-3.6823, interpMode=RCIM_Cubic, leaveTangent=-3.6823, tangentMode=RCTM_User, time=0.45, value=1}
         {arriveTangent=0.0656, interpMode=RCIM_Cubic, leaveTangent=0.0656, tangentMode=RCTM_User, time=1, value=0}
        ]}}})
    ScaleA=true
    ScaleRGB=false
   )
   /SetVariables(Particles.MeshOrientation=&MultiplyQuaternion(
     Quaternion A=/Niagara/DynamicInputs/TypeConversions/MakeCustomQuaternionFromBool(
      Bool:NiagaraBool=false
      True Quat:Quat4f=$Output.SkeletalMeshLocation.SampledRotation
      False Quat:Quat4f=$Particles.SampleSkeletalMeshSkeleton.SampledRotation
     )
     Quaternion B=&MakeQuatFromAxisAngle(Axis=<1 0 0>, AngleInDegrees=&Multiply_Float(A=$Particles.Initial.MaterialRandom, B=360))
    )) {isSetParameters=true, module=SetVariables_#6}
  ]
  rendererClasses [&NiagaraMeshRendererProperties]
  renderers [{rendererClass=&NiagaraMeshRendererProperties, rendererIndex=0}]
 }
 Fire {
  emitterUpdateScript [
   /SetVariables(Emitter.AccelerationRamp:NiagaraFloat=$System.AccelerationRamp) {isSetParameters=true, module=SetVariables_#3}
   &EmitterState(Enable Distance Culling=true, MaxDistance=1024, Module.bUseMaxDistance=true, Module.bUseMinDistance=false, Scalability Mode=::NewEnumerator1)
   &SpawnRate(SpawnRate=&Lerp_Float(A=15, B=30, Alpha=$Emitter.AccelerationRamp), Use Spawn Probability=false)
  ]
  particleSpawnScript [
   &InitializeParticle(
    Color=&LerpLinearColors(StartColor=<rgba 0 0.45 1 2>, EndColor=<rgba 0 8.0974 10 1>, LerpFactor=$Emitter.AccelerationRamp)
    Lifetime=&Lerp_Float(
     A=&UniformRangedFloat(Minimum=0.8, Maximum=1)
     B=&UniformRangedFloat(Minimum=0.5, Maximum=0.7)
     Alpha=$Emitter.AccelerationRamp
    )
    Module.Write Color=true
    Module.Write Lifetime=true
    Module.Write Mass=false
    Module.Write Material Random=true
    Module.Write Position=false
    Module.Write Scale=false
    Module.Write SpriteRotation=true
    Module.Write SpriteSize=true
    Position=/Niagara/Functions/Localspace/SimulationPosition()
    Sprite Rotation=&UniformRangedFloat(Minimum=-180, Maximum=180)
    Sprite Size=&Vector2DFromFloat(Value=&Lerp_Float(
      A=&UniformRangedFloat(Minimum=35, Maximum=55)
      B=&UniformRangedFloat(Minimum=50, Maximum=70)
      Alpha=$Emitter.AccelerationRamp
     ))
   )
   &SphereLocation(Hemisphere Y=true, MaskSpawnGroup=false, Module.Override Randomness=false, Sphere Radius=1, Surface Only=false)
   /SetVariables(Particles.SphereVector:Vector3f=&Multiply_VectorByFloat(Vector=$Output.SphereLocation.SphereNormal, Float=21.21)) {isSetParameters=true}
   &DynamicMaterialParameters(
    Index 0 Param 2=&UniformRangedFloat(Minimum=0, Maximum=1)
    Index 0 Param 3=1
    Index 0 Param 4=&UniformRangedFloat(Minimum=1, Maximum=2)
    Module.Param0WriteEnabled=false
    Module.Param1WriteEnabled=true
    Module.Param2WriteEnabled=false
    Module.Param3WriteEnabled=true
   ) {module=DynamicMaterialParameters002}
   &AddVelocity(Velocity=&FN_FrontEndVector3Modifier(
     Gameplay=&Lerp_Vector(
      A=&UniformRangedVector(Minimum=<-25 0 0>, Maximum=<0 1 1>)
      B=&UniformRangedVector(Minimum=<-60 0 0>, Maximum=<-120 1 1>)
      Alpha=$Emitter.AccelerationRamp
     )
     Front End=&Lerp_Vector(
      A=&UniformRangedVector(Minimum=<25 0 0>, Maximum=<0 1 1>)
      B=&UniformRangedVector(Minimum=<60 0 0>, Maximum=<120 1 1>)
      Alpha=$Emitter.AccelerationRamp
     )
    ))
   &CameraOffset(Camera Offset Amount=&Lerp_Float(A=25, B=6, Alpha=$Emitter.AccelerationRamp))
  ]
  particleUpdateScript [
   &ParticleState()
   &SkeletalMeshLocation(
    Apply Sampled Orientation to Mesh Particles=true
    Bone / Socket Index=1
    Inherit Velocity=false
    MaskSpawnGroup=false
    Mesh Sampling Type=::NewEnumerator5
    Skeletal Mesh=$User.SkeletalMesh
    Socket Sampling Mode=::NewEnumerator5
   )
   /SetVariables(Particles.Position=&Add_Vector(A=$Particles.SphereVector, B=$Output.SkeletalMeshLocation.SampledPosition)) {isSetParameters=true}
   &Drag(Use Linear Drag=true, Use Rotational Drag=false)
   &SolveForcesAndVelocity(Clamp Velocity=false, Limit Acceleration=false)
   &SubUVAnimation(Number Of Frames=64)
   &ScaleSpriteSize(Scale Factor=&Vector2DFromFloat(Value=&FloatFromCurve(FloatCurve=DataInterface() {propertyValues {Curve {keys [
          {arriveTangent=7.5764, interpMode=RCIM_Cubic, time=0, value=0.02}
          {arriveTangent=1.2063, interpMode=RCIM_Cubic, leaveTangent=1.2063, tangentMode=RCTM_User, time=0.3, value=0.99}
          <at 1 0.5 RCIM_Cubic>
         ]}}})))
   &DynamicMaterialParameters(
    Index 0 Param 1=&FloatFromCurve(FloatCurve=DataInterface() {propertyValues {Curve {keys [{arriveTangent=2.7551, interpMode=RCIM_Cubic, leaveTangent=2.7551, tangentMode=RCTM_User, time=0, value=0}, <at 0.7 1>]}}})
    Index 0 Param 2=1
    Index 0 Param 3=1
    Index 0 Param 4=1
    Module.Param0WriteEnabled=true
    Module.Param1WriteEnabled=true
    Module.Param2WriteEnabled=true
    Module.Param3WriteEnabled=false
   ) {module=DynamicMaterialParameters001}
   &FN_OffsetPositionByOrientationAxis(Offset Axis:Vector3f=<25 -6 0>, Position:NiagaraPosition=$Particles.Position) {module=OffsetPositionByMeshOrientation}
   &ScaleColor(
    Scale Alpha=&FloatFromCurve(FloatCurve=DataInterface() {propertyValues {Curve {keys [
         {arriveTangent=8.7245, interpMode=RCIM_Cubic, leaveTangent=8.7245, tangentMode=RCTM_User, time=0, value=0}
         <at 0.2 1 RCIM_Cubic>
         {arriveTangent=-0.5281, interpMode=RCIM_Cubic, leaveTangent=-0.5281, tangentMode=RCTM_User, time=1, value=0}
        ]}}})
    ScaleA=true
    ScaleRGB=true
   )
  ]
  rendererClasses [&NiagaraSpriteRendererProperties]
  renderers [{rendererClass=&NiagaraSpriteRendererProperties}]
 }
 Front_Flare_Large {
  emitterUpdateScript [
   /Game/Effects/Niagara/Modules/Math/FN_Integrator/FN_Integrator_v2(
    Clamp=true
    DecayRate=HlslExpression() {hlslExpression="User.ForwardDot < .1 ? Engine.DeltaTime : 0"}
    Input=HlslExpression() {hlslExpression="User.ForwardDot > .1 ? Engine.DeltaTime : 0"}
    Max=10
    Namespace:ENiagaraNameSpace=::NewEnumerator1
   )
   /SetVariables(Emitter.AccelerationRamp:NiagaraFloat=&FN_FrontEndFloatModifier(
     Gameplay=/Game/Effects/Niagara/DynamicInputs/Math/FN_Fit(
      Clamp:NiagaraBool=true
      Value:NiagaraFloat=$Emitter.FN_Integrator_v2.Attr
      SourceMin:NiagaraFloat=1
      SourceMax:NiagaraFloat=2
      DestMin:NiagaraFloat=0
      DestMax:NiagaraFloat=1
     )
     Front End=1
    )) {isSetParameters=true, module=SetVariables_#3}
   &EmitterState(
    Inactive Response=::NewEnumerator1
    Life Cycle Mode=::NewEnumerator1
    Loop Behavior=::NewEnumerator1
    Loop Delay=&FN_FrontEndFloatModifier(Gameplay=1, Front End=0.1)
    Loop Duration Mode=::NewEnumerator1
    UseLoopDelay=false
   )
   &SpawnBurst_Instantaneous(Spawn Count=1, Use Spawn Probability=false)
  ]
  particleSpawnScript [
   /Game/Effects/Niagara/Modules/Groups/FN_SetSpawnGroup_v1/FN_SetSpawnGroup(Emitter Bone Count=2)
   &InitializeParticle(
    Lifetime=0.6
    Mesh Scale=&FN_FrontEndVector3Modifier(Gameplay=<1 1 1>, Front End=<0.75 0.7 0.7>)
    Module.Write Color=false
    Module.Write Lifetime=false
    Module.Write Mass=false
    Module.Write Material Random=true
    Module.Write Position=false
    Module.Write Scale=true
    Module.Write SpriteRotation=false
    Module.Write SpriteSize=false
    Sprite Size=<5 5>
   )
  ]
  particleUpdateScript [
   &ParticleState(Kill Particles When Lifetime Has Elapsed=false, Let Infinitely Lived Particles Die When Emitter Deactivates=true)
   &SkeletalMeshLocation(
    Apply Sampled Orientation to Mesh Particles=true
    Inherit Velocity=false
    MaskSpawnGroup=false
    Mesh Sampling Type=::NewEnumerator5
    Sampled Position Offset=&FN_FrontEndVector3Modifier(Gameplay=<5 0 0>, Front End=<10 0 0>)
    Skeletal Mesh=$User.SkeletalMesh
    Socket Sampling Mode=::NewEnumerator5
   )
   &SolveForcesAndVelocity(Clamp Velocity=false, Limit Acceleration=false)
   &Color(Color=&LerpLinearColors(StartColor=<rgba 1.414 0 4 1>, EndColor=<rgba 6 0 12 0.666>, LerpFactor=HlslExpression() {hlslExpression="float(Particles.SpawnGroup)"}))
   &ScaleColor(Color Value To Scale=$Particles.Color, Scale Alpha=$Emitter.AccelerationRamp, ScaleA=true, ScaleRGB=false)
  ]
  rendererClasses [&NiagaraMeshRendererProperties]
  renderers [{rendererClass=&NiagaraMeshRendererProperties, rendererIndex=0}]
 }
 Lights {
  emitterUpdateScript [
   /SetVariables(Emitter.AccelerationRamp:NiagaraFloat=$System.AccelerationRamp) {isSetParameters=true}
   &EmitterState(
    Enable Distance Culling=true
    Enable Visibility Culling=true
    MaxDistance=1024
    Module.bUseMaxDistance=true
    Module.bUseMinDistance=false
    Scalability Mode=::NewEnumerator1
   )
   &SpawnRate(Use Spawn Probability=false)
  ]
  particleSpawnScript [
   &InitializeParticle(
    Lifetime=2
    Module.Write Color=false
    Module.Write Lifetime=true
    Module.Write Mass=true
    Module.Write Material Random=true
    Module.Write Position=true
    Module.Write Scale=false
    Module.Write SpriteRotation=false
    Module.Write SpriteSize=true
    Position=&Add_Vector(A=&SimulationPosition(), B=<0 0 50>)
    Sprite Size=&Vector2DFromFloat(Value=10)
   )
   /SetVariables(
    Particles.lightoffsetX:NiagaraFloat=&UniformAOrBFloat(A=-20, B=20)
    Particles.lightoffsetY:NiagaraFloat=&UniformAOrBFloat(A=-20, B=30)
   ) {isSetParameters=true}
   /SetVariables() {isSetParameters=true}
  ]
  particleUpdateScript [
   &ParticleState()
   &Color(
    Color=&LerpLinearColors(
     StartColor=&ColorFromCurve(ColorCurve=DataInterface() {propertyValues {
        AlphaCurve {keys []}
        BlueCurve {keys [<at -0.005 0>,<at 0.509 6 RCIM_Cubic>,<at 1.0372 0>]}
        GreenCurve {keys [<at -0.005 0>,<at 0.509 0 RCIM_Cubic>,<at 1.0372 0>]}
        RedCurve {keys [<at -0.005 0>,<at 0.509 0.6873 RCIM_Cubic>,<at 1.0372 0>]}
       }})
     EndColor=&LerpLinearColors(
      StartColor=&ColorFromCurve(ColorCurve=DataInterface() {propertyValues {
         AlphaCurve {keys []}
         BlueCurve {keys [<at -0.005 0>,<at 0.509 6 RCIM_Cubic>,<at 1.0372 0>]}
         GreenCurve {keys [<at -0.005 0>,<at 0.509 4.5556 RCIM_Cubic>,<at 1.0372 0>]}
         RedCurve {keys [<at -0.005 0>,<at 0.509 0 RCIM_Cubic>,<at 1.0372 0>]}
        }})
      EndColor=&ColorFromCurve(ColorCurve=DataInterface() {propertyValues {
         AlphaCurve {keys []}
         BlueCurve {keys [<at -0.005 0>,<at 0.509 6 RCIM_Cubic>,<at 1.0372 0>]}
         GreenCurve {keys [<at -0.005 0>,<at 0.509 0 RCIM_Cubic>,<at 1.0372 0>]}
         RedCurve {keys [<at -0.005 0>,<at 0.509 3.3778 RCIM_Cubic>,<at 1.0372 0>]}
        }})
      LerpFactor=&ClampFloat(Float=&Multiply_Float(A=&Subtract_Float(A=0.5, B=$Particles.Initial.MaterialRandom), B=10))
     )
     LerpFactor=$Emitter.AccelerationRamp
    )
    Scale Alpha=&FloatFromCurve(FloatCurve=DataInterface() {propertyValues {Curve {keys [<at 0 0 RCIM_Cubic>,<at 0.5 1 RCIM_Cubic>,<at 1 0>]}}})
   )
   &SampleSkeletalMeshSkeleton(Mesh:NiagaraDataInterfaceSkeletalMesh=$User.SkeletalMesh)
   &SkeletalMeshSkeletonLocation(Module.Inherit Velocity=false, Module.Use Sampled Bone as Ribbon ID=true)
   &FN_OffsetPositionByOrientationAxis(
    Offset Axis:Vector3f=&MakeVector(X=$Particles.Initial.lightoffsetX, Y=$Particles.Initial.lightoffsetY, Z=100)
    Position:NiagaraPosition=$Particles.Position
   ) {module=OffsetPositionByMeshOrientation}
   /SetVariables() {enabled=false, isSetParameters=true}
  ]
  rendererClasses [&NiagaraSpriteRendererProperties, &NiagaraLightRendererProperties]
  renderers [
   {rendererClass=&NiagaraSpriteRendererProperties}
   {rendererClass=&NiagaraLightRendererProperties, rendererIndex=1}
  ]
 }
 Mini_L {
  emitterUpdateScript [
   /SetVariables(Emitter.AccelerationRamp:NiagaraFloat=$System.AccelerationRamp) {isSetParameters=true, module=SetVariables_#4}
   &EmitterState(Enable Distance Culling=true, MaxDistance=1024, Module.bUseMaxDistance=true, Module.bUseMinDistance=false, Scalability Mode=::NewEnumerator1)
   &SpawnRate(
    Spawn Probability=&Lerp_Float(A=0.6, B=1, Alpha=$Emitter.AccelerationRamp)
    SpawnRate=&Lerp_Float(A=12, B=14, Alpha=$Emitter.AccelerationRamp)
    Use Spawn Probability=true
   )
  ]
  particleSpawnScript [
   &FN_SetSpawnGroup(Emitter Bone Count=2)
   &InitializeParticle(
    Color=&LerpLinearColors(StartColor=<rgba 5 3 5 0.2>, EndColor=<rgba 5 3.3 5 0.2>, LerpFactor=$Emitter.AccelerationRamp)
    Lifetime=&Lerp_Float(A=0.4, B=0.3, Alpha=$Emitter.AccelerationRamp)
    Mesh Scale=&Lerp_Vector(
     A=&Lerp_Vector(A=<0.25 0.09 0.09>, B=<1.5 0.11 0.11>, Alpha=$Emitter.AccelerationRamp)
     B=&Lerp_Vector(A=<0.25 0.07 0.07>, B=<1.5 0.095 0.095>, Alpha=$Emitter.AccelerationRamp)
     Alpha=HlslExpression() {hlslExpression=Particles.SpawnGroup}
    )
    Module.Write Color=true
    Module.Write Lifetime=true
    Module.Write Mass=true
    Module.Write Material Random=true
    Module.Write Position=false
    Module.Write Scale=true
    Module.Write SpriteRotation=true
    Module.Write SpriteSize=false
    Sprite Rotation=&UniformRangedFloat(Minimum=-0.2, Maximum=0.2)
    Sprite Size=<5 5>
   )
   /SetVariables(Particles.SpawnGroup=&Add_Integer(A=$Particles.SpawnGroup, B=2)) {isSetParameters=true}
  ]
  particleUpdateScript [
   &ParticleState()
   &SkeletalMeshLocation(
    Apply Sampled Orientation to Mesh Particles=true
    Bone / Socket Index=$Particles.SpawnGroup
    Inherit Velocity=false
    MaskSpawnGroup=false
    Mesh Sampling Type=::NewEnumerator5
    Skeletal Mesh=$User.SkeletalMesh
    Socket Sampling Mode=::NewEnumerator5
   ) {module=SkeletalMeshLocation001}
   &ScaleMeshSize(Scale Factor=&Multiply_Vector(
     A=&MakeVector(
      X=&FloatFromCurve(FloatCurve=DataInterface() {propertyValues {Curve {keys [<at 0 0.8 RCIM_Cubic>, {arriveTangent=2.6748, interpMode=RCIM_Cubic, leaveTangent=2.6748, tangentMode=RCTM_User, time=0.75, value=1.5}]}}})
      Y=0.9
      Z=0.9
     )
     B=&VectorFromFloat(Value=&FloatFromCurve(FloatCurve=DataInterface() {propertyValues {Curve {keys [<at 0 0 RCIM_Constant>,<at 0.1596 1>,<at 1 1>]}}}))
    )) {module=ScaleMeshSize001}
   &ScaleColor(
    Scale Alpha=&FloatFromCurve(FloatCurve=DataInterface() {propertyValues {Curve {keys [
         <at 0 0>
         {arriveTangent=1.6667, leaveTangent=1.6667, time=0.4, value=1}
         {arriveTangent=-3.6823, interpMode=RCIM_Cubic, leaveTangent=-3.6823, tangentMode=RCTM_User, time=0.45, value=1}
         {arriveTangent=0.0656, interpMode=RCIM_Cubic, leaveTangent=0.0656, tangentMode=RCTM_User, time=1, value=0}
        ]}}})
    ScaleA=true
    ScaleRGB=true
   )
   /SetVariables(Particles.MeshOrientation=&MultiplyQuaternion(
     Quaternion A=$Output.SkeletalMeshLocation001.SampledRotation
     Quaternion B=&MakeQuatFromAxisAngle(Axis=<1 0 0>, AngleInDegrees=&Multiply_Float(A=$Particles.Initial.MaterialRandom, B=360))
    )) {isSetParameters=true, module=SetVariables_#6}
  ]
  rendererClasses [&NiagaraMeshRendererProperties]
  renderers [{rendererClass=&NiagaraMeshRendererProperties, rendererIndex=0}]
 }
 Purple_Main {
  bEnabled=false
  emitterUpdateScript [
   /SetVariables(Emitter.AccelerationRamp:NiagaraFloat=&Divide_Float(A=&Subtract_Float(A=$User.SizeScale, B=0.25), B=0.75)) {isSetParameters=true, module=SetVariables_#7}
   &EmitterState(Inactive Response=::NewEnumerator1, Life Cycle Mode=::NewEnumerator1, Loop Duration=1, UseLoopDelay=false)
   &SpawnRate(SpawnRate=&Lerp_Float(A=2, B=5, Alpha=$Emitter.AccelerationRamp), Use Spawn Probability=false)
  ]
  particleSpawnScript [
   /SetVariables(Particles.MaterialRandom:NiagaraFloat=&UniformRangedFloat(Minimum=0, Maximum=1)) {isSetParameters=true, module=SetVariables_#5}
   &InitializeParticle(
    Color=&LerpLinearColors(StartColor=<rgba 1 0 1 2>, EndColor=<rgba 10 0 1 1>, LerpFactor=$Emitter.AccelerationRamp)
    Lifetime=&Lerp_Float(A=1.1, B=0.6, Alpha=$Emitter.AccelerationRamp)
    Material Random:NiagaraFloat=$Particles.MaterialRandom
    Mesh Scale=&FN_FrontEndVector3Modifier(
     Gameplay=&Multiply_Vector(A=<1 1 1>, B=&Lerp_Vector(A=<1 0.77 0.77>, B=<1.777 0.89 0.89>, Alpha=$Emitter.AccelerationRamp))
     Front End=&Multiply_Vector(A=<1 1 1>, B=&Lerp_Vector(A=<1 0.5 0.5>, B=<1.777 0.65 0.65>, Alpha=$Emitter.AccelerationRamp))
    )
    Module.Write Color=true
    Module.Write Lifetime=true
    Module.Write Mass=false
    Module.Write Material Random=true
    Module.Write Position=false
    Module.Write Scale=true
    Module.Write SpriteRotation=true
    Module.Write SpriteSize=false
    Position=&SimulationPosition()
    Sprite Rotation=&UniformRangedFloat(Minimum=-0.2, Maximum=0.2)
    Sprite Size=<5 5>
   )
  ]
  particleUpdateScript [
   &ParticleState()
   &SkeletalMeshLocation(
    Apply Sampled Orientation to Mesh Particles=true
    Bone / Socket Index=1
    Inherit Velocity=false
    MaskSpawnGroup=false
    Mesh Sampling Type=::NewEnumerator5
    Sampled Position Offset=<0 0 -0.5>
    Skeletal Mesh=$User.SkeletalMesh
    Socket Sampling Mode=::NewEnumerator5
   ) {enabled=false}
   &SampleSkeletalMeshSkeleton(
    Mesh:NiagaraDataInterfaceSkeletalMesh=$User.SkeletalMesh
    Spawn Mode:ENiagaraSkeletalSpawning=::NewEnumerator1
    Specific Bone Index=0
    Specific Socket Index=1
   )
   &SkeletalMeshSkeletonLocation(Module.Inherit Velocity=false, Module.Use Sampled Bone as Ribbon ID=true)
   &FN_OffsetPositionByOrientationAxis(
    Offset Axis:Vector3f=&FN_FrontEndVector3Modifier(Gameplay=<120 -5 -20>, Front End=<90 0 -10>)
    Position:NiagaraPosition=$Particles.Position
   ) {module=OffsetPositionByMeshOrientation}
   &ScaleMeshSize(Scale Factor=&VectorFromCurve(VectorCurve=DataInterface() {propertyValues {
       XCurve {keys [
         {arriveTangent=4.2301, interpMode=RCIM_Cubic, leaveTangent=4.2301, tangentMode=RCTM_User, time=0, value=0.2}
         {arriveTangent=0.0169, interpMode=RCIM_Cubic, leaveTangent=0.0169, tangentMode=RCTM_User, time=1, value=2}
        ]}
       YCurve {keys [
         {arriveTangent=0.0463, interpMode=RCIM_Cubic, leaveTangent=0.0463, tangentMode=RCTM_User, time=0, value=0.98}
         {arriveTangent=-1.8909, interpMode=RCIM_Cubic, leaveTangent=-1.8909, tangentMode=RCTM_User, time=1, value=0.5}
        ]}
       ZCurve {keys [<at 0 0.98 RCIM_Cubic>, {arriveTangent=-1.8424, interpMode=RCIM_Cubic, leaveTangent=-1.8424, tangentMode=RCTM_User, time=1, value=0.5}]}
      }})) {module=ScaleMeshSize001}
   &ScaleColor(
    Scale Alpha=&FloatFromCurve(FloatCurve=DataInterface() {propertyValues {Curve {keys [
         <at 0 0 RCIM_Cubic>
         <at 0.3 1>
         {arriveTangent=0.0062, interpMode=RCIM_Cubic, leaveTangent=0.0062, tangentMode=RCTM_User, time=0.6172, value=0.9817}
         {arriveTangent=-3.9755, interpMode=RCIM_Cubic, leaveTangent=-3.9755, tangentMode=RCTM_User, time=1, value=0}
        ]}}})
    ScaleA=true
    ScaleRGB=true
   )
   /SetVariables(Particles.MeshOrientation=&MultiplyQuaternion(
     Quaternion A=&MakeCustomQuaternionFromBool(
      Bool:NiagaraBool=false
      True Quat:Quat4f=$Output.SkeletalMeshLocation.SampledRotation
      False Quat:Quat4f=$Particles.SampleSkeletalMeshSkeleton.SampledRotation
     )
     Quaternion B=&MakeQuatFromAxisAngle(
      Axis=&FN_FrontEndVector3Modifier(Gameplay=<1 0 0>, Front End=<0 0 0>)
      AngleInDegrees=&Multiply_Float(A=$Particles.Initial.MaterialRandom, B=360)
     )
    )) {isSetParameters=true, module=SetVariables_#6}
  ]
  rendererClasses [&NiagaraMeshRendererProperties]
  renderers [{rendererClass=&NiagaraMeshRendererProperties, rendererIndex=0}]
 }
 Purple_Main001 {
  emitterUpdateScript [
   /SetVariables(Emitter.AccelerationRamp:NiagaraFloat=$System.AccelerationRamp) {isSetParameters=true, module=SetVariables_#7}
   &EmitterState(Inactive Response=::NewEnumerator1, Life Cycle Mode=::NewEnumerator1, Loop Duration=1, UseLoopDelay=false)
   &SpawnRate(SpawnRate=&Lerp_Float(A=4, B=6, Alpha=$Emitter.AccelerationRamp), Use Spawn Probability=false)
  ]
  particleSpawnScript [
   &FN_SetSpawnGroup(Emitter Bone Count=2)
   &InitializeParticle(
    Color=&LerpLinearColors(
     StartColor=&LerpLinearColors(StartColor=<rgba 1 0 1 2>, EndColor=<rgba 10 0 1 1>, LerpFactor=$Emitter.AccelerationRamp)
     EndColor=&LerpLinearColors(StartColor=<rgba 1 1 1 1.333>, EndColor=<rgba 2.5 2.5 2.5 1>, LerpFactor=$Emitter.AccelerationRamp)
     LerpFactor=HlslExpression() {hlslExpression=Particles.SpawnGroup}
    )
    Lifetime=&Lerp_Float(
     A=&Lerp_Float(A=1.1, B=0.6, Alpha=$Emitter.AccelerationRamp)
     B=&Lerp_Float(A=0.7, B=1, Alpha=$Emitter.AccelerationRamp)
     Alpha=HlslExpression() {hlslExpression=Particles.SpawnGroup}
    )
    Material Random:NiagaraFloat=$Particles.MaterialRandom
    Mesh Scale=&Lerp_Vector(
     A=&FN_FrontEndVector3Modifier(
      Gameplay=&Multiply_Vector(A=<1 1 1>, B=&Lerp_Vector(A=<1 0.77 0.77>, B=<1.777 0.89 0.89>, Alpha=$Emitter.AccelerationRamp))
      Front End=&Multiply_Vector(A=<1 0.85 0.85>, B=&Lerp_Vector(A=<1 0.5 0.5>, B=<1.777 0.65 0.65>, Alpha=$Emitter.AccelerationRamp))
     )
     B=&Multiply_Vector(A=<0.8 0.666 0.666>, B=&Lerp_Vector(A=<1.333 1 1>, B=<1.666 0.87 0.87>, Alpha=$Emitter.AccelerationRamp))
     Alpha=HlslExpression() {hlslExpression=Particles.SpawnGroup}
    )
    Module.Write Color=true
    Module.Write Lifetime=true
    Module.Write Mass=false
    Module.Write Material Random=true
    Module.Write Position=false
    Module.Write Scale=true
    Module.Write SpriteRotation=true
    Module.Write SpriteSize=false
    Position=&SimulationPosition()
    Sprite Rotation=&UniformRangedFloat(Minimum=-0.2, Maximum=0.2)
    Sprite Size=<5 5>
   )
  ]
  particleUpdateScript [
   &ParticleState()
   &SkeletalMeshLocation(
    Apply Sampled Orientation to Mesh Particles=true
    Bone / Socket Index=1
    Inherit Velocity=false
    MaskSpawnGroup=false
    Mesh Sampling Type=::NewEnumerator5
    Sampled Position Offset=<0 0 -0.5>
    Skeletal Mesh=$User.SkeletalMesh
    Socket Sampling Mode=::NewEnumerator5
   )
   &FN_OffsetPositionByOrientationAxis(
    Offset Axis:Vector3f=&Lerp_Vector(
     A=&FN_FrontEndVector3Modifier(Gameplay=<120 -5 -20>, Front End=<90 0 -10>)
     B=<65 -4 -10>
     Alpha=HlslExpression() {hlslExpression=Particles.SpawnGroup}
    )
    Position:NiagaraPosition=$Particles.Position
   ) {module=OffsetPositionByMeshOrientation}
   &ScaleColor(
    Scale Alpha=&FloatFromCurve(FloatCurve=DataInterface() {propertyValues {Curve {keys [
         <at 0 0 RCIM_Cubic>
         <at 0.3 1>
         {arriveTangent=0.0062, interpMode=RCIM_Cubic, leaveTangent=0.0062, tangentMode=RCTM_User, time=0.6172, value=0.9817}
         {arriveTangent=-3.9755, interpMode=RCIM_Cubic, leaveTangent=-3.9755, tangentMode=RCTM_User, time=1, value=0}
        ]}}})
    ScaleA=true
    ScaleRGB=true
   )
   /SetVariables(Particles.MeshOrientation=&MultiplyQuaternion(
     Quaternion A=$Output.SkeletalMeshLocation.SampledRotation
     Quaternion B=&MakeQuatFromAxisAngle(
      Axis=&FN_FrontEndVector3Modifier(Gameplay=<1 0 0>, Front End=<0 0 0>)
      AngleInDegrees=&Multiply_Float(A=$Particles.Initial.MaterialRandom, B=360)
     )
    )) {isSetParameters=true, module=SetVariables_#6}
   /SetVariables(Particles.PurpleScale:Vector3f=&Lerp_Vector(
     A=&Multiply_Vector(
      A=&Multiply_Vector(
       A=$Particles.Initial.Scale
       B=&VectorFromCurve(VectorCurve=DataInterface() {propertyValues {
          XCurve {keys [
            {arriveTangent=4.2301, interpMode=RCIM_Cubic, leaveTangent=4.2301, tangentMode=RCTM_User, time=0, value=0.2}
            {arriveTangent=0.0169, interpMode=RCIM_Cubic, leaveTangent=0.0169, tangentMode=RCTM_User, time=1, value=2}
           ]}
          YCurve {keys [
            {arriveTangent=0.0463, interpMode=RCIM_Cubic, leaveTangent=0.0463, tangentMode=RCTM_User, time=0, value=0.98}
            {arriveTangent=-1.8909, interpMode=RCIM_Cubic, leaveTangent=-1.8909, tangentMode=RCTM_User, time=1, value=0.5}
           ]}
          ZCurve {keys [<at 0 0.98 RCIM_Cubic>, {arriveTangent=-1.8424, interpMode=RCIM_Cubic, leaveTangent=-1.8424, tangentMode=RCTM_User, time=1, value=0.5}]}
         }})
      )
      B=<1 1 1>
     )
     B=<0 0 0>
     Alpha=HlslExpression() {hlslExpression=Particles.SpawnGroup}
    )) {isSetParameters=true}
   /SetVariables(Particles.BlueScale:Vector3f=&Lerp_Vector(
     A=<0 0 0>
     B=&Multiply_Vector(
      A=&Multiply_Vector(
       A=$Particles.Initial.Scale
       B=&VectorFromCurve(VectorCurve=DataInterface() {propertyValues {
          XCurve {keys [
            {arriveTangent=0.0368, interpMode=RCIM_Cubic, leaveTangent=0.0368, tangentMode=RCTM_User, time=0, value=0.1}
            {arriveTangent=0.0169, interpMode=RCIM_Cubic, leaveTangent=0.0169, tangentMode=RCTM_User, time=1, value=2}
           ]}
          YCurve {keys [
            {arriveTangent=0.0463, interpMode=RCIM_Cubic, leaveTangent=0.0463, tangentMode=RCTM_User, time=-0.0013, value=0.6712}
            {interpMode=RCIM_Cubic, leaveTangent=-1.8909, time=0.9977, value=0.1545}
           ]}
          ZCurve {keys [<at -0.0081 0.686 RCIM_Cubic>, {interpMode=RCIM_Cubic, leaveTangent=-1.8424, time=0.9977, value=0.1545}]}
         }})
      )
      B=&VectorFromCurve(
       VectorCurve=DataInterface() {propertyValues {
         XCurve {keys [<at 0.2506 0.7 RCIM_Cubic>,<at 0.9973 1.2845>]}
         YCurve {keys [<at 0.2506 0.7 RCIM_Cubic>,<at 0.989 1.1924 RCIM_Cubic>]}
         ZCurve {keys [<at 0.2506 0.7 RCIM_Cubic>,<at 0.989 1.1924 RCIM_Cubic>]}
        }}
       CurveIndex=$Emitter.AccelerationRamp
      )
     )
     Alpha=HlslExpression() {hlslExpression="float(Particles.SpawnGroup)"}
    )) {isSetParameters=true}
  ]
  rendererClasses [&NiagaraMeshRendererProperties]
  renderers [
   {rendererClass=&NiagaraMeshRendererProperties, rendererIndex=0}
   {rendererClass=&NiagaraMeshRendererProperties, rendererIndex=1}
  ]
 }
 Speedlines {
  emitterUpdateScript [
   /SetVariables(Emitter.AccelerationRamp:NiagaraFloat=$System.AccelerationRamp) {isSetParameters=true, module=SetVariables_#3}
   &EmitterState(Enable Distance Culling=true, MaxDistance=2048, Module.bUseMaxDistance=true, Module.bUseMinDistance=false, Scalability Mode=::NewEnumerator1)
   &SpawnRate(
    SpawnRate=&FN_FrontEndFloatModifier(
     Gameplay=&Lerp_Float(A=0, B=&UniformRangedFloat(Minimum=150, Maximum=210), Alpha=$Emitter.AccelerationRamp)
     Front End=0
    )
    Use Spawn Probability=true
   )
  ]
  particleSpawnScript [
   &InitializeParticle(
    Color=&MakeLinearColorFromVectorAndFloat(
     "Vector (RGB)"=&MakeVectorFromLinearColorRGB(Color=&LerpLinearColors(StartColor=<rgba 0 0.8097 1 1>, EndColor=<rgba 0.4629 0 1 1>, LerpFactor=&UniformAOrBFloat(A=0)))
     "Float (Alpha)"=&UniformRangedFloat(Minimum=0.05, Maximum=0.4)
    )
    Lifetime=&UniformRangedFloat(Minimum=0.2, Maximum=0.35)
    Module.Write Color=true
    Module.Write Lifetime=true
    Module.Write Mass=true
    Module.Write Material Random=true
    Module.Write Position=false
    Module.Write Scale=false
    Module.Write SpriteRotation=false
    Module.Write SpriteSize=true
    Position=&SimulationPosition()
    Sprite Size=&UniformRangedVector2D(Minimum=<1.5 4>, Maximum=<3 6>)
   )
   &CylinderLocation(
    Cylinder Radius=&UniformRangedFloat(Minimum=150, Maximum=210)
    MaskSpawnGroup=false
    Module.Enable Lathe Profile=false
    Module.Override Randomness=false
    Module.Surface Only=true
    Offset=<0 -500 30>
    Orientation Axis:ENiagaraOrientationAxis=::XAxis
    Use Endcaps In Surface Only Mode=false
   )
   &AddVelocity(Velocity=<0 2000 0>)
   &AddVelocityFromPoint(Module.Use Falloff Distance=false, Velocity Strength=&UniformRangedFloat(Minimum=350, Maximum=450))
  ]
  particleUpdateScript [
   &ParticleState()
   &SolveForcesAndVelocity(Clamp Velocity=false, Limit Acceleration=false)
   &ScaleSpriteSize(Scale Factor=&Vector2DFromCurve(
     Vector2Curve=DataInterface() {propertyValues {XCurve {keys [<at 0 0.5 RCIM_Cubic>,<at 0.5 1 RCIM_Cubic>]}, YCurve {keys [<at 0 0 RCIM_Cubic>,<at 0.5 1 RCIM_Cubic>]}}}
     ScaleCurve=<1 1>
    ))
   &ScaleSpriteSizeBySpeed(Max Scale Factor=<1 135>, Min Scale Factor=<0 0.05>, Velocity Threshold=1)
   &ScaleColor(
    Scale Alpha=&FloatFromCurve(FloatCurve=DataInterface() {propertyValues {Curve {keys [<at 0 0 RCIM_Cubic>, <at 0.2 1>, {arriveTangent=-2.221, interpMode=RCIM_Cubic, leaveTangent=-2.221, tangentMode=RCTM_User, time=1, value=0}]}}})
    ScaleA=true
    ScaleRGB=true
   )
  ]
  rendererClasses [&NiagaraSpriteRendererProperties]
  renderers [{rendererClass=&NiagaraSpriteRendererProperties}]
 }
 crystal_shards {
  emitterUpdateScript [
   /SetVariables(Emitter.AccelerationRamp:NiagaraFloat=$System.AccelerationRamp) {isSetParameters=true, module=SetVariables_#3}
   &EmitterState(Enable Distance Culling=true, MaxDistance=1024, Module.bUseMaxDistance=true, Module.bUseMinDistance=false, Scalability Mode=::NewEnumerator1)
   &SpawnRate(
    Spawn Probability=0.4
    SpawnRate=&Lerp_Float(A=9, B=48, Alpha=$Emitter.AccelerationRamp)
    Use Spawn Probability=true
   )
  ]
  particleSpawnScript [
   &FN_SetSpawnGroup(Emitter Bone Count=2)
   &InitializeParticle(
    Color=&LerpLinearColors(
     StartColor=&MakeLinearColorFromVectorAndFloat(
      "Vector (RGB)"=&MakeVectorFromLinearColorRGB(Color=<rgba 0 80.9736 100 1>)
      "Float (Alpha)"=&Lerp_Float(
       A=&UniformRangedFloat(Minimum=0.8, Maximum=1.2)
       B=&UniformRangedFloat(Minimum=1.5, Maximum=3)
       Alpha=$Emitter.AccelerationRamp
      )
     )
     EndColor=<rgba 50 0 50 1>
     LerpFactor=HlslExpression() {hlslExpression=Particles.SpawnGroup}
    )
    Lifetime=&Lerp_Float(
     A=&UniformRangedFloat(Minimum=0.9, Maximum=1.35)
     B=&UniformRangedFloat(Minimum=0.3, Maximum=0.6)
     Alpha=HlslExpression() {hlslExpression=Particles.SpawnGroup}
    )
    Mesh Scale=&Lerp_Vector(
     A=&Multiply_Vector(
      A=<0.1 0.1 0.025>
      B=&Lerp_Vector(
       A=&VectorFromFloat(Value=&UniformRangedFloat(Minimum=0.666, Maximum=1.35))
       B=&UniformRangedVector(Minimum=<5 0.2 0.2>, Maximum=<25 0.8 0.8>)
       Alpha=$Emitter.AccelerationRamp
      )
     )
     B=&Multiply_Vector(
      A=<0.1 0.1 0.025>
      B=&Lerp_Vector(
       A=&VectorFromFloat(Value=&UniformRangedFloat(Minimum=0.666, Maximum=1.35))
       B=&UniformRangedVector(Minimum=<10 0.2 0.2>, Maximum=<65 0.8 0.8>)
       Alpha=$Emitter.AccelerationRamp
      )
     )
     Alpha=HlslExpression() {hlslExpression=Particles.SpawnGroup}
    )
    Module.Write Color=true
    Module.Write Lifetime=true
    Module.Write Mass=false
    Module.Write Material Random=true
    Module.Write Position=true
    Module.Write Scale=true
    Module.Write SpriteRotation=false
    Module.Write SpriteSize=false
    Position=&Add_Vector(A=&SimulationPosition(), B=<12 90 -35>)
    Sprite Size=<5 5>
   )
   /Niagara/Modules/Spawn/Orientation/V2/InitialMeshOrientation(Use Orientation Vector=true, Use Rotation Vector=false)
   &AddVelocity(Velocity=&FN_FrontEndVector3Modifier(
     Gameplay=&MakeVector(
      X=0
      Y=&Lerp_Float(
       A=&UniformRangedFloat(Minimum=60, Maximum=120)
       B=&UniformRangedFloat(Minimum=900, Maximum=1500)
       Alpha=$Emitter.AccelerationRamp
      )
      Z=0
     )
     Front End=&MakeVector(
      X=0
      Y=&Lerp_Float(
       A=&UniformRangedFloat(Minimum=-60, Maximum=-120)
       B=&UniformRangedFloat(Minimum=-900, Maximum=-1500)
       Alpha=$Emitter.AccelerationRamp
      )
      Z=0
     )
    ))
   &SkeletalMeshLocation(
    Bone / Socket Index=1
    Inherit Velocity=false
    MaskSpawnGroup=false
    Mesh Sampling Type=::NewEnumerator5
    Skeletal Mesh=$User.SkeletalMesh
    Socket Sampling Mode=::NewEnumerator5
   )
   &CylinderLocation(
    Cylinder Height=10
    Cylinder Radius=&Lerp_Float(
     A=&UniformRangedFloat(Minimum=0, Maximum=20)
     B=&UniformRangedFloat(Minimum=30, Maximum=32)
     Alpha=HlslExpression() {hlslExpression=Particles.SpawnGroup}
    )
    MaskSpawnGroup=false
    Module.Enable Lathe Profile=false
    Module.Override Randomness=false
    Module.Surface Only=true
    Orientation Axis:ENiagaraOrientationAxis=::YAxis
    Surface Only Band Thickness:NiagaraFloat=&Lerp_Float(A=5, B=2, Alpha=HlslExpression() {hlslExpression=Particles.SpawnGroup})
    Use Endcaps In Surface Only Mode=false
   )
  ]
  particleUpdateScript [
   &ParticleState()
   /Niagara/Modules/Update/Forces/CurlNoiseForce(Noise Frequency=5, Noise Strength=150, Pan Noise=false, Random Seed=1)
   &SolveForcesAndVelocity(Clamp Velocity=false, Limit Acceleration=false)
   &ScaleMeshSize(Scale Factor=&Lerp_Vector(
     A=&VectorFromFloat(Value=&FloatFromCurve(FloatCurve=DataInterface() {propertyValues {Curve {keys [
           <at 0 1>
           {arriveTangent=-0.5431, interpMode=RCIM_Cubic, leaveTangent=-0.5431, tangentMode=RCTM_User, time=0.3, value=1}
           {arriveTangent=-5.5854, interpMode=RCIM_Cubic, leaveTangent=-5.5854, tangentMode=RCTM_User, time=1, value=0}
          ]}}}))
     B=&VectorFromFloat(Value=&FloatFromCurve(FloatCurve=DataInterface() {propertyValues {Curve {keys [
           <at 0 1>
           {arriveTangent=-0.5431, interpMode=RCIM_Cubic, leaveTangent=-0.5431, tangentMode=RCTM_User, time=0.6, value=1}
           {arriveTangent=-5.5854, interpMode=RCIM_Cubic, leaveTangent=-5.5854, tangentMode=RCTM_User, time=1, value=0}
          ]}}}))
     Alpha=HlslExpression() {hlslExpression=Particles.SpawnGroup}
    ))
   &ScaleColor(
    Scale Alpha=&FloatFromCurve(FloatCurve=DataInterface() {propertyValues {Curve {keys [<at 0 1>,<at 1 0>]}}})
    Scale RGB=&VectorFromFloat(Value=&FloatFromCurve(FloatCurve=DataInterface() {propertyValues {Curve {keys [<at 0 1>,<at 1 0>]}}}))
    ScaleA=true
    ScaleRGB=true
   )
  ]
  rendererClasses [&NiagaraMeshRendererProperties]
  renderers [{rendererClass=&NiagaraMeshRendererProperties, rendererIndex=0}]
 }
 crystal_shards_SpeedDebris {
  emitterUpdateScript [
   /SetVariables(Emitter.AccelerationRamp:NiagaraFloat=$System.AccelerationRamp) {isSetParameters=true, module=SetVariables_#3}
   &EmitterState()
   &SpawnRate(
    Spawn Probability=0.3
    SpawnRate=&FN_FrontEndFloatModifier(
     Gameplay=&Lerp_Float(
      A=&UniformRangedFloat(Minimum=5, Maximum=7)
      B=&UniformRangedFloat(Minimum=12, Maximum=16)
      Alpha=$Emitter.AccelerationRamp
     )
     Front End=0
    )
    Use Spawn Probability=true
   )
  ]
  particleSpawnScript [
   &InitializeParticle(
    Color=&LerpLinearColors(
     StartColor=<rgba 170.0004 0 300 1>
     EndColor=&LerpLinearColors(
      StartColor=<rgba 0 299.9998 300 1>
      EndColor=<rgba 225.0003 0 300 1>
      LerpFactor=&UniformRangedFloat(Minimum=0, Maximum=1)
     )
     LerpFactor=$Emitter.AccelerationRamp
    )
    Lifetime=&UniformRangedFloat(Minimum=0.3, Maximum=0.9)
    Module.Write Color=true
    Module.Write Lifetime=true
    Module.Write Mass=false
    Module.Write Material Random=true
    Module.Write Position=false
    Module.Write Scale=false
    Module.Write SpriteRotation=false
    Module.Write SpriteSize=true
    Sprite Size=&Vector2DFromFloat(Value=&Multiply_Float(A=20, B=&UniformRangedFloat(Minimum=0.5, Maximum=1)))
   )
   &InitialMeshOrientation(Use Orientation Vector=true, Use Rotation Vector=false)
   &SkeletalMeshLocation(
    Inherit Velocity=false
    MaskSpawnGroup=false
    Mesh Sampling Type=::NewEnumerator5
    Skeletal Mesh=$User.SkeletalMesh
    Socket Sampling Mode=::NewEnumerator5
   )
   &CylinderLocation(
    Cylinder Height=0
    Cylinder Radius=&UniformRangedFloat(Minimum=50, Maximum=60)
    MaskSpawnGroup=false
    Module.Enable Lathe Profile=false
    Module.Override Randomness=false
    Module.Surface Only=true
    Orientation Axis:ENiagaraOrientationAxis=::YAxis
    Use Endcaps In Surface Only Mode=false
   )
  ]
  particleUpdateScript [
   &ParticleState()
   &AccelerationForce(
    Acceleration=&FN_FrontEndVector3Modifier(Gameplay=<0 2000 0>, Front End=<0 -2000 0>)
    Coordinate Space=::Simulation
   )
   &SolveForcesAndVelocity(Clamp Velocity=false, Limit Acceleration=false)
   &ScaleColor(
    Scale Alpha=&FloatFromCurve(FloatCurve=DataInterface() {propertyValues {Curve {keys [<at 0 1>,<at 1 0>]}}})
    Scale RGB=&VectorFromFloat(Value=&FloatFromCurve(FloatCurve=DataInterface() {propertyValues {Curve {keys [<at 0 1 RCIM_Cubic>, {arriveTangent=-1.8097, interpMode=RCIM_Cubic, leaveTangent=-1.8097, tangentMode=RCTM_User, time=1, value=0}]}}}))
    ScaleA=true
    ScaleRGB=true
   )
   &ScaleSpriteSize(Scale Factor=&Vector2DFromFloat(Value=&Multiply_Float(A=$Particles.NormalizedAge, B=&OneMinusFloat(Float=$Particles.NormalizedAge))))
   &ScaleSpriteSizeBySpeed(
    Max Scale Factor=<0.5 20>
    Velocity Threshold:NiagaraFloat=&Lerp_Float(A=1800, B=1200, Alpha=$Emitter.AccelerationRamp)
   )
  ]
  rendererClasses [&NiagaraSpriteRendererProperties]
  renderers [{rendererClass=&NiagaraSpriteRendererProperties}]
 }
}
summary {
 systemName=NS_Glider_SpaceWanderer_Female_Deploy_Optimizaed
 emitters [
  {bEnabled=false, emitterName=Purple_Main, rendererClasses [&NiagaraMeshRendererProperties]}
  {bEnabled=false, emitterName=Blue_Main, rendererClasses [&NiagaraMeshRendererProperties]}
  {emitterName=Front_Flare_Large, rendererClasses [&NiagaraMeshRendererProperties]}
  {emitterName=Lights, rendererClasses [&NiagaraSpriteRendererProperties, &NiagaraLightRendererProperties]}
  {emitterName=Blue_Decal, rendererClasses [&NiagaraSpriteRendererProperties]}
  {emitterName=Speedlines, rendererClasses [&NiagaraSpriteRendererProperties]}
  {emitterName=Ash, simTarget=GPUComputeSim, rendererClasses [&NiagaraSpriteRendererProperties]}
  {emitterName=Mini_L, rendererClasses [&NiagaraMeshRendererProperties]}
  {emitterName=Fire, rendererClasses [&NiagaraSpriteRendererProperties]}
  {emitterName=crystal_shards_SpeedDebris, rendererClasses [&NiagaraSpriteRendererProperties]}
  {emitterName=crystal_shards, rendererClasses [&NiagaraMeshRendererProperties]}
  {emitterName=Purple_Main001, rendererClasses [&NiagaraMeshRendererProperties]}
 ]
 userVariables [
  User.bIsFrontEnd {type=NiagaraBool}
  User.bIsFrontEndPreview {type=NiagaraBool, defaultValue=true}
  User.FEVec {type=Vector3f, description="", defaultValue {struct=Vector3f, value=<0 0 0>}}
  User.ForwardDot {type=NiagaraFloat, defaultValue=0, description=""}
  User.Open {type=NiagaraFloat, defaultValue=0.15, description=""}
  User.PlayerSpeed {type=NiagaraFloat, defaultValue=500, description=""}
  User.RibbonIDOffset {type=NiagaraInt32}
  User.SizeScale {type=NiagaraFloat, defaultValue=0, description=""}
  User.SkeletalMesh {
   type=NiagaraDataInterfaceSkeletalMesh
   defaultValue {
    struct=NiagaraExt_VariableValue_DataInterface
    value {dataInterface=&self:NiagaraDataInterfaceSkeletalMesh_2}
   }
  }
 ]
}
